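# M_Wheel_ETB is the material for the ETB cosmetic vehicle wheel, adding a masked rim glow to the shared MF_WheelParent_Simple base.
# An Emissive_Mask texture, constant color and intensity parameter form a glow, blended over the parent's emissive by inverted vertex-color red.
# The updated material attributes then pass through the base vehicle damage and de-rez transition functions.
# It supports skeletal meshes and mesh deformers.
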Material /VehicleCosmetics/Wheels/ETB/Materials/M_Wheel_ETB
graph {
 properties {
  bAutomaticallySetUsageInEditor=false
  bUseMaterialAttributes=true
  bUsedWithMeshDeformer=true
  bUsedWithSkeletalMesh=true
  pixelDepthOffsetMode=PDOM_Legacy
  usedShadingModels=Default Lit
 }
 nodes [
  n1 = MaterialFunctionCall() {materialFunction=&MF_WheelParent_Simple}
  n2 = GetMaterialAttributes(None=@n1.MaterialAttributes) {attributeGetTypes [#1]}
  n3 = LinearInterpolate(A=@n2.Emissive Color, B=NamedRerouteUsage(), Alpha=OneMinus(None=@n1.VertexColorR))
  MaterialFunctionCall(In=MaterialFunctionCall(Material=SetMaterialAttributes(MaterialAttributes=@n2.MaterialAttributes, Emissive Color=@n3) {attributeSetTypes [#1]}) {materialFunction=&MF_BaseVehicle_Damage}.Result) {materialFunction=&MF_VehicleDeRez_Transition}
  n4 = TextureSampleParameter2D() {parameterName=Emissive_Mask, samplerType=SAMPLERTYPE_LinearColor, texture=/VehicleCosmetics/Wheels/ETB/Textures/T_ETB_Rim_M}
  n5 = Constant4Vector() {constant=<rgba 0.0802 0.5333 0.4742 1>}
  n6 = MaterialFunctionCall(R=@n5.R, G=@n5.R, B=@n5.R) {materialFunction=&AppendMany}
  NamedRerouteDeclaration(None=Multiply(A=Multiply(A=@n4.R, B=@n6.RGB), B=ScalarParameter() {group=2_Rim Properties, parameterName=Emissive_Intensity})) {name=Emissive, nodeColor=<rgba 1 0.75 0 1>}
 ]
}
