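# Masked unlit skeletal-mesh material for the ETB LEOT vehicle wheel cosmetic.
# A wheel texture with stretched U tiling pans continuously along U, and its alpha drives the opacity mask, with no input wired to emissive color.
# The material attributes then pass through the shared base vehicle damage function and into the vehicle de-rez transition function.
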Material /VehicleCosmetics/Wheels/ETB/Materials/M_Wheel_ETB_LEOT
graph {
 properties {
  bAutomaticallySetUsageInEditor=false
  bUseMaterialAttributes=true
  bUsedWithSkeletalMesh=true
  blendMode=BLEND_Masked
  shadingModel=MSM_Unlit
  usedShadingModels=Unlit
 }
 nodes [
  n1 = TextureSample(UVs=Panner(Coordinate=TextureCoordinate() {uTiling=0.25}) {speedX=-0.25}) {texture=/VehicleCosmetics/Wheels/ETB/Textures/Wheel_ETB_LEOT_BC}
  MaterialFunctionCall(In=MaterialFunctionCall(Material=SetMaterialAttributes!(MaterialAttributes, Emissive Color, Opacity Mask=@n1.A) {attributeSetTypes [#1, #2]}) {materialFunction=&MF_BaseVehicle_Damage}.Result) {materialFunction=&MF_VehicleDeRez_Transition}
 ]
}
